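# Physics asset for the De Las Muertos cape back accessory, giving its tassel bones simulated secondary motion.
# A capsule on backpack_root anchors five parallel four-bone tassel chains (left, mid-left, mid, mid-right, right), with one capsule per bone.
# Each chain's first joint has tight swing and twist limits at the root, while lower joints are looser, so the cape bends more further down.
# No constraints link neighboring chains sideways.
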PhysicsAsset /BRCosmetics/Accessories/FORT_Capes/M_MED_DeLasMuertos/Meshes/M_MED_DeLasMuertos_Cape_Physics
bodies {
 backpack_root [{shapeType=Capsule, length=20.8413, radius=2.2931, center=<0.015 -5.879 -6.2468>, rotation=<rot 0 -90.2224 -89.9951>}]
 tassle_01_l [{shapeType=Capsule, length=2.5672, radius=3.3013, shapeName=tassle_01_l_capsule, center=<4.3624 0.1792 -0.1146>, rotation=<rot 90 0 0>}]
 tassle_01_mid [{shapeType=Capsule, length=0.9744, radius=4.0763, shapeName=tassle_01_mid_capsule, center=<2.3018 0 0>}]
 tassle_01_mid_l [{shapeType=Capsule, length=2.3172, radius=1.8148, shapeName=tassle_01_mid_l_capsule, center=<4.0732 -0.0513 0.0384>, rotation=<rot 90 0 0>}]
 tassle_01_mid_r [{shapeType=Capsule, length=3.0703, radius=1.072, shapeName=tassle_01_mid_r_capsule, center=<4.9456 0.1622 -0.0094>, rotation=<rot 90 0 0>}]
 tassle_01_r [{shapeType=Capsule, length=2.4906, radius=0.75, shapeName=tassle_01_r_capsule, center=<3.6703 -0.1224 -0.0315>, rotation=<rot 90 0 0>}]
 tassle_02_l [{shapeType=Capsule, length=4.1701, radius=1.8678, shapeName=tassle_02_l_capsule, center=<3.5561 0.0779 -0.0077>, rotation=<rot 90 0 0>}]
 tassle_02_mid [{shapeType=Capsule, length=0.1778, radius=4.8938, shapeName=tassle_02_mid_capsule, center=<2.3018 0 0>}]
 tassle_02_mid_l [{shapeType=Capsule, length=0, radius=4.3977, shapeName=tassle_02_mid_l_capsule, center=<3.6206 -0.0199 0.0202>, rotation=<rot 90 0 0>}]
 tassle_02_mid_r [{shapeType=Capsule, length=2.4196, radius=1.8884, shapeName=tassle_02_mid_r_capsule, center=<4.2098 0.0559 0.0543>, rotation=<rot 90 0 0>}]
 tassle_02_r [{shapeType=Capsule, length=2.9792, radius=1.1176, shapeName=tassle_02_r_capsule, center=<4.301 0.0152 0.0032>, rotation=<rot 90 0 0>}]
 tassle_03_l [{shapeType=Capsule, length=2.0827, radius=4.2947, shapeName=tassle_03_l_capsule, center=<3.8795 -0.0698 0.0136>, rotation=<rot 90 0 0>}]
 tassle_03_mid [{shapeType=Capsule, length=0.2881, radius=5.0174, shapeName=tassle_03_mid_capsule, center=<2.3018 0 0>}]
 tassle_03_mid_l [{shapeType=Capsule, length=0.038, radius=4.5136, shapeName=tassle_03_mid_l_capsule, center=<3.8798 -0.1515 0.1269>, rotation=<rot 90 0 0>}]
 tassle_03_mid_r [{shapeType=Capsule, length=2.5382, radius=2.0185, shapeName=tassle_03_mid_r_capsule, center=<4.4106 -0.1501 0.1139>, rotation=<rot 90 0 0>}]
 tassle_03_r [{shapeType=Capsule, length=2.7892, radius=1.2824, shapeName=tassle_03_r_capsule, center=<3.9585 0.0183 0.0793>, rotation=<rot 90 0 0>}]
 tassle_04_l [{shapeType=Capsule, length=2.76, radius=0.5, shapeName=tassle_04_l_capsule, center=<1.8227 0 0>, rotation=<rot 90 0 0>}]
 tassle_04_mid [{shapeType=Capsule, length=0, radius=3.005, shapeName=tassle_04_mid_capsule, center=<2.3018 0 0>}]
 tassle_04_mid_l [{shapeType=Capsule, length=0.255, radius=1.255, shapeName=tassle_04_mid_l_capsule, center=<1.8592 0 0>, rotation=<rot 90 0 0>}]
 tassle_04_mid_r [{shapeType=Capsule, length=0.76, radius=0.75, shapeName=tassle_04_mid_r_capsule, center=<2.4009 0 0>, rotation=<rot 90 0 0>}]
 tassle_04_r [{shapeType=Capsule, length=0.26, radius=0.75, shapeName=tassle_04_r_capsule, center=<2.0089 0 0>, rotation=<rot 90 0 0>}]
}
constraints [
 {bone1Name=tassle_01_l, bone2Name=backpack_root, swing1LimitDegrees=5, swing2LimitDegrees=5, twistLimitDegrees=5}
 {bone1Name=tassle_02_l, bone2Name=tassle_01_l, swing1LimitDegrees=15, swing2LimitDegrees=15, twistLimitDegrees=15}
 {bone1Name=tassle_03_l, bone2Name=tassle_02_l, swing1LimitDegrees=15, swing2LimitDegrees=15, twistLimitDegrees=15}
 {bone1Name=tassle_04_l, bone2Name=tassle_03_l, swing1LimitDegrees=15, swing2LimitDegrees=15, twistLimitDegrees=15}
 {bone1Name=tassle_01_mid, bone2Name=backpack_root, swing1LimitDegrees=5, swing2LimitDegrees=5, twistLimitDegrees=5}
 {bone1Name=tassle_02_mid, bone2Name=tassle_01_mid, swing1LimitDegrees=15, swing2LimitDegrees=15, twistLimitDegrees=15}
 {bone1Name=tassle_03_mid, bone2Name=tassle_02_mid, swing1LimitDegrees=15, swing2LimitDegrees=15, twistLimitDegrees=15}
 {bone1Name=tassle_04_mid, bone2Name=tassle_03_mid, swing1LimitDegrees=15, swing2LimitDegrees=15, twistLimitDegrees=15}
 {bone1Name=tassle_01_mid_l, bone2Name=backpack_root, swing1LimitDegrees=5, swing2LimitDegrees=5, twistLimitDegrees=5}
 {bone1Name=tassle_02_mid_l, bone2Name=tassle_01_mid_l, swing1LimitDegrees=15, swing2LimitDegrees=15, twistLimitDegrees=15}
 {bone1Name=tassle_03_mid_l, bone2Name=tassle_02_mid_l, swing1LimitDegrees=15, swing2LimitDegrees=15, twistLimitDegrees=15}
 {bone1Name=tassle_04_mid_l, bone2Name=tassle_03_mid_l, swing1LimitDegrees=15, swing2LimitDegrees=15, twistLimitDegrees=15}
 {bone1Name=tassle_01_mid_r, bone2Name=backpack_root, swing1LimitDegrees=5, swing2LimitDegrees=5, twistLimitDegrees=5}
 {bone1Name=tassle_02_mid_r, bone2Name=tassle_01_mid_r, swing1LimitDegrees=15, swing2LimitDegrees=15, twistLimitDegrees=15}
 {bone1Name=tassle_03_mid_r, bone2Name=tassle_02_mid_r, swing1LimitDegrees=15, swing2LimitDegrees=15, twistLimitDegrees=15}
 {bone1Name=tassle_04_mid_r, bone2Name=tassle_03_mid_r, swing1LimitDegrees=15, swing2LimitDegrees=15, twistLimitDegrees=15}
 {bone1Name=tassle_01_r, bone2Name=backpack_root, swing1LimitDegrees=5, swing2LimitDegrees=5, twistLimitDegrees=5}
 {bone1Name=tassle_02_r, bone2Name=tassle_01_r, swing1LimitDegrees=15, swing2LimitDegrees=15, twistLimitDegrees=15}
 {bone1Name=tassle_03_r, bone2Name=tassle_02_r, swing1LimitDegrees=15, swing2LimitDegrees=15, twistLimitDegrees=15}
 {bone1Name=tassle_04_r, bone2Name=tassle_03_r, swing1LimitDegrees=15, swing2LimitDegrees=15, twistLimitDegrees=15}
]
